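# Sound cue for the burst shotgun's drop impact, playing when the weapon lands after being dropped.
# A single wave player feeds the BurstShotgun_Land sound straight to the output, routed through the Fort_Loot_Collect sound class.
# The cue overrides attenuation for a short-range falloff and overrides concurrency to cap simultaneous instances, preventing new plays once the limit is reached.
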SoundCue /FlipperGameplay/Items/Weapons/BurstShotgun/Audio/BurstShotgun_DropImpact_Cue
root=WavePlayer_0
nodes [{id=WavePlayer_0, properties {soundWaveAssetPtr=&/BurstShotgun_Land}}]
properties {
 bOverrideAttenuation=true
 bOverrideConcurrency=true
 maxDistance=1000
 soundClassObject=&Fort_Loot_Collect
 volumeMultiplier=1
 attenuationOverrides {falloffDistance=1000, attenuationShapeExtents {x=0}}
 concurrencyOverrides {maxCount=3, resolutionRule=PreventNew}
}
