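# Sound cue for the Burst Shotgun's first-person firing sound.
# A random node picks one of five equally weighted fire recordings and feeds a modulator that varies playback, with the minimum pitch held at unshifted.
# The cue uses first-person weapon attenuation, a ranged-fire sound class and a shared first-person gunfire concurrency set.
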SoundCue /FlipperGameplay/Items/Weapons/BurstShotgun/Audio/BurstShotgun_Fire_1P_Cue
root=Modulator_0
nodes [
 {id=Modulator_0, children [Random_0], properties {pitchMin=1}}
 {id=Random_0, children [WavePlayer_2,WavePlayer_3,WavePlayer_4,WavePlayer_5,WavePlayer_6], properties {weights [1,1,1,1,1]}}
 {id=WavePlayer_2, properties {soundWaveAssetPtr=&/BurstShotgun_Fire_1P_01}}
 {id=WavePlayer_3, properties {soundWaveAssetPtr=&/BurstShotgun_Fire_1P_02}}
 {id=WavePlayer_4, properties {soundWaveAssetPtr=&/BurstShotgun_Fire_1P_03}}
 {id=WavePlayer_5, properties {soundWaveAssetPtr=&/BurstShotgun_Fire_1P_04}}
 {id=WavePlayer_6, properties {soundWaveAssetPtr=&/BurstShotgun_Fire_1P_05}}
]
properties {
 attenuationSettings=&Fort_Weapon_Projectile_FirstPerson_Athena_Att
 maxDistance=2097152
 soundClassObject=&Fort_WeaponRanged_Fire_1P_PSM
 volumeMultiplier=1.1
 attenuationOverrides {falloffDistance=5000, attenuationShapeExtents {x=1000}}
 concurrencyOverrides {bLimitToOwner=true, maxCount=3, resolutionRule=StopOldest}
 concurrencySet [&Weapon_Gunfire_FirstPerson_Conc]
}
