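# Sound cue for the third-person firing sound of the Burst Shotgun, layered by distance.
# A mixer blends close, medium and far branches, each picking randomly from five wave variants, then passing through a modulator, its own distance attenuation and a sound class override.
# The far layer uses a separate distant-fire sound class, while close and medium share the third-person fire class.
# Local and global gunfire concurrency sets are assigned to limit overlapping shots.
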SoundCue /FlipperGameplay/Items/Weapons/BurstShotgun/Audio/BurstShotgun_Fire_Cue
root=Mixer_0
nodes [
 {id=Attenuation_0, children [Modulator_0], properties {attenuationSettings=&Fort_Weapon_Projectile_Close_Athena_Att}}
 {
  id=Attenuation_1
  children [Modulator_1]
  properties {attenuationSettings=&Fort_Weapon_Projectile_Medium_Mono_Athena_Att}
 }
 {
  id=Attenuation_2
  children [Modulator_2]
  properties {attenuationSettings=&Fort_Weapon_Projectile_Distant_Mono_Athena_Att}
 }
 {id=Mixer_0, children [SoundClass_0,SoundClass_6,SoundClass_1], properties {inputVolume [0.7,0.6,0.4]}}
 {id=Modulator_0, children [Random_3], properties {pitchMin=1}}
 {id=Modulator_1, children [Random_4], properties {volumeMax=2, volumeMin=2}}
 {id=Modulator_2, children [Random_5], properties {volumeMax=1.5, volumeMin=1.5}}
 {id=Random_3, children [WavePlayer_0,WavePlayer_1,WavePlayer_2,WavePlayer_3,WavePlayer_4], properties {weights [1,1,1,1,1]}}
 {id=Random_4, children [WavePlayer_5,WavePlayer_6,WavePlayer_7,WavePlayer_8,WavePlayer_9], properties {weights [1,1,1,1,1]}}
 {id=Random_5, children [WavePlayer_10, WavePlayer_11, WavePlayer_12, WavePlayer_14, WavePlayer_15], properties {weights [1,1,1,1,1]}}
 {
  id=SoundClass_0
  children [Attenuation_0]
  properties {soundClassOverride=&Fort_WeaponRanged_Fire_3P_PSM}
 }
 {
  id=SoundClass_1
  children [Attenuation_2]
  properties {soundClassOverride=&Fort_WeaponRanged_Fire_Distant}
 }
 {
  id=SoundClass_6
  children [Attenuation_1]
  properties {soundClassOverride=&Fort_WeaponRanged_Fire_3P_PSM}
 }
 {id=WavePlayer_0, properties {soundWaveAssetPtr=&/BurstShotgun_Fire_3P_Close_01}}
 {id=WavePlayer_1, properties {soundWaveAssetPtr=&/BurstShotgun_Fire_3P_Close_02}}
 {id=WavePlayer_10, properties {soundWaveAssetPtr=&/BurstShotgun_Fire_3P_Far_01}}
 {id=WavePlayer_11, properties {soundWaveAssetPtr=&/BurstShotgun_Fire_3P_Far_02}}
 {id=WavePlayer_12, properties {soundWaveAssetPtr=&/BurstShotgun_Fire_3P_Far_03}}
 {id=WavePlayer_14, properties {soundWaveAssetPtr=&/BurstShotgun_Fire_3P_Far_04}}
 {id=WavePlayer_15, properties {soundWaveAssetPtr=&/BurstShotgun_Fire_3P_Far_05}}
 {id=WavePlayer_2, properties {soundWaveAssetPtr=&/BurstShotgun_Fire_3P_Close_03}}
 {id=WavePlayer_3, properties {soundWaveAssetPtr=&/BurstShotgun_Fire_3P_Close_04}}
 {id=WavePlayer_4, properties {soundWaveAssetPtr=&/BurstShotgun_Fire_3P_Close_05}}
 {id=WavePlayer_5, properties {soundWaveAssetPtr=&/BurstShotgun_Fire_3P_Med_01}}
 {id=WavePlayer_6, properties {soundWaveAssetPtr=&/BurstShotgun_Fire_3P_Med_02}}
 {id=WavePlayer_7, properties {soundWaveAssetPtr=&/BurstShotgun_Fire_3P_Med_03}}
 {id=WavePlayer_8, properties {soundWaveAssetPtr=&/BurstShotgun_Fire_3P_Med_04}}
 {id=WavePlayer_9, properties {soundWaveAssetPtr=&/BurstShotgun_Fire_3P_Med_05}}
]
properties {
 maxDistance=27500
 soundClassObject=&Fort_WeaponRanged_Fire_3P_PSM
 volumeMultiplier=0.9
 attenuationOverrides {falloffDistance=5000, attenuationShapeExtents {x=1000}}
 concurrencyOverrides {bLimitToOwner=true, maxCount=3, resolutionRule=StopOldest}
 concurrencySet [&Weapon_Gunfire_Local_Conc, &Weapon_Gunfire_Global_Conc]
}
