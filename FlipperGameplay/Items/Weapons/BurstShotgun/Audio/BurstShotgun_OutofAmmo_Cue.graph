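# Sound cue for the burst shotgun's out-of-ammo sound.
# A random node picks one of four equally weighted NoAmmo waves, feeding a modulator that adds slight volume variation with fixed pitch.
# The cue plays at reduced volume on a third-person ranged-weapon fire sound class.
# Its concurrency override allows two simultaneous instances and stops the oldest beyond that.
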SoundCue /FlipperGameplay/Items/Weapons/BurstShotgun/Audio/BurstShotgun_OutofAmmo_Cue
root=Modulator_0
nodes [
 {id=Modulator_0, children [Random_0], properties {pitchMax=1, pitchMin=1, volumeMax=1.1, volumeMin=1}}
 {id=Random_0, children [WavePlayer_5,WavePlayer_6,WavePlayer_7,WavePlayer_0], properties {weights [1,1,1,1]}}
 {id=WavePlayer_0, properties {soundWaveAssetPtr=&/BurstShotgun_NoAmmo_04}}
 {id=WavePlayer_5, properties {soundWaveAssetPtr=&/BurstShotgun_NoAmmo_01}}
 {id=WavePlayer_6, properties {soundWaveAssetPtr=&/BurstShotgun_NoAmmo_02}}
 {id=WavePlayer_7, properties {soundWaveAssetPtr=&/BurstShotgun_NoAmmo_03}}
]
properties {
 bOverrideConcurrency=true
 maxDistance=2097152
 soundClassObject=&Fort_WeaponRanged_Fire_3P_PSM
 volumeMultiplier=0.3
 attenuationOverrides {falloffDistance=1600}
 concurrencyOverrides {maxCount=2, resolutionRule=StopOldest}
}
