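# Sound cue that plays the pickup sound for the burst shotgun weapon in FlipperGameplay.
# A single wave player feeds the BurstShotgun_Pickup wave to the output at reduced volume.
# It uses the shared weapon pickup attenuation and loot collect sound class.
# A concurrency override allows one instance at a time and blocks new plays while it is active.
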SoundCue /FlipperGameplay/Items/Weapons/BurstShotgun/Audio/BurstShotgun_Pickup_Cue1
root=WavePlayer_0
nodes [{id=WavePlayer_0, properties {soundWaveAssetPtr=&/BurstShotgun_Pickup}}]
properties {
 attenuationSettings=&Fort_Weapon_Pickup_Att
 bOverrideConcurrency=true
 maxDistance=3500
 soundClassObject=&Fort_Loot_Collect
 volumeMultiplier=0.2769
 concurrencyOverrides {maxCount=1, resolutionRule=PreventNew}
}
