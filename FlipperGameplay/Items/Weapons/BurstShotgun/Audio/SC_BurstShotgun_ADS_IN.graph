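# SoundCue for the burst shotgun's aim-down-sights-in foley sound.
# A weighted Random node picks one of three shotgun ADS-in waves, favoring the first two, and feeds a Modulator that varies pitch per play.
# The cue plays at reduced volume through the weapon foley sound class and borrows the weapon reload attenuation settings.
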SoundCue /FlipperGameplay/Items/Weapons/BurstShotgun/Audio/SC_BurstShotgun_ADS_IN
root=Modulator_0
nodes [
 {id=Modulator_0, children [Random_1], properties {pitchMax=1.2}}
 {id=Random_1, children [WavePlayer_1,WavePlayer_2,WavePlayer_3], properties {weights [1.136,1.2493,0.64]}}
 {id=WavePlayer_1, properties {soundWaveAssetPtr=&/Weapon_Shotgun_ADS_In_01}}
 {id=WavePlayer_2, properties {soundWaveAssetPtr=&/Weapon_Shotgun_ADS_In_02}}
 {id=WavePlayer_3, properties {soundWaveAssetPtr=&/Weapon_Shotgun_ADS_In_03}}
]
properties {
 attenuationSettings=&Fort_Weapon_Reload_Att
 maxDistance=2250
 pitchMultiplier=1.0242
 soundClassObject=&Fort_WeaponFoley
 volumeMultiplier=0.6
}
