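# Sound cue for the burst shotgun's aim-down-sights exit foley.
# It plays a single heavy-shot ADS-out wave through a Modulator that randomizes pitch upward for variation.
# The cue runs at reduced volume in the weapon foley sound class and uses the shared weapon reload attenuation.
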SoundCue /FlipperGameplay/Items/Weapons/BurstShotgun/Audio/SC_BurstShotgun_ADS_OUT
root=Modulator_0
nodes [
 {id=Modulator_0, children [WavePlayer_1], properties {pitchMax=1.2, pitchMin=1}}
 {id=WavePlayer_1, properties {soundWaveAssetPtr=&/Weapon_HEAVYSHOT_ADS_OUT_02}}
]
properties {
 attenuationSettings=&Fort_Weapon_Reload_Att
 maxDistance=2250
 pitchMultiplier=0.976
 soundClassObject=&Fort_WeaponFoley
 volumeMultiplier=0.6
}
