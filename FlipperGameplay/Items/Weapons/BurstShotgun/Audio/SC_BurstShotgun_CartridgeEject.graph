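# Sound cue for the Burst Shotgun's cartridge eject foley.
# A random node picks one of four equally weighted ReloadEnd waves, then a modulator randomizes its pitch upward.
# The cue lowers overall volume and pitch, uses the shared weapon reload attenuation, and is routed to the weapon foley sound class.
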SoundCue /FlipperGameplay/Items/Weapons/BurstShotgun/Audio/SC_BurstShotgun_CartridgeEject
root=Modulator_0
nodes [
 {id=Modulator_0, children [Random_1], properties {pitchMax=1.2, pitchMin=1}}
 {id=Random_1, children [WavePlayer_0,WavePlayer_1,WavePlayer_6,WavePlayer_7], properties {weights [1,1,1,1]}}
 {id=WavePlayer_0, properties {soundWaveAssetPtr=&/BurstShotgun_ReloadEnd_04}}
 {id=WavePlayer_1, properties {soundWaveAssetPtr=&/BurstShotgun_ReloadEnd_03}}
 {id=WavePlayer_6, properties {soundWaveAssetPtr=&/BurstShotgun_ReloadEnd_02}}
 {id=WavePlayer_7, properties {soundWaveAssetPtr=&/BurstShotgun_ReloadEnd_01}}
]
properties {
 attenuationSettings=&Fort_Weapon_Reload_Att
 maxDistance=2250
 pitchMultiplier=0.928
 soundClassObject=&Fort_WeaponFoley
 volumeMultiplier=0.4
}
