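# Sound cue for the burst shotgun's cartridge impact, set up as quiet weapon foley with reload attenuation.
# A random node feeds a modulator that varies pitch upward slightly on each play.
# All three equally weighted random branches point to the same wave player, so only one sample ever plays and variation comes from pitch alone.
# A limit-to-one, prevent-new concurrency rule keeps impacts from stacking.
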SoundCue /FlipperGameplay/Items/Weapons/BurstShotgun/Audio/SC_BurstShotgun_CartridgeImpact
root=Modulator_0
nodes [
 {id=Modulator_0, children [Random_0], properties {pitchMax=1.2, pitchMin=1}}
 {id=Random_0, children [WavePlayer_3,WavePlayer_3,WavePlayer_3], properties {weights [1,1,1]}}
 {
  id=WavePlayer_3
  properties {soundWaveAssetPtr=/FlipperGameplay/Items/Weapons/BurstShotgun/Audio/Srnnd_Versions/sw_BurstShotgun_1p_Cartridge_Impact_02}
 }
]
properties {
 attenuationSettings=&Fort_Weapon_Reload_Att
 maxDistance=2250
 soundClassObject=&Fort_WeaponFoley
 volumeMultiplier=0.2
 concurrencySet [/Game/Sounds/Concurrency_Rules/LimitToOnePreventNew]
}
