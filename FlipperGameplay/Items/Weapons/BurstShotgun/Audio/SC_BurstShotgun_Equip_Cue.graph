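# SoundCue for the Burst Shotgun's equip sound, classed as weapon foley.
# A single wave player reuses a FirePetal AutoShotgun equip wave and feeds a Modulator that applies a fixed volume boost.
# The cue lowers overall volume and raises pitch, using the shared weapon reload attenuation.
# Its concurrency override prevents new instances when the limit is reached.
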SoundCue /FlipperGameplay/Items/Weapons/BurstShotgun/Audio/SC_BurstShotgun_Equip_Cue
root=Modulator_0
nodes [
 {id=Modulator_0, children [WavePlayer_6], properties {volumeMax=2, volumeMin=2}}
 {id=WavePlayer_6, properties {soundWaveAssetPtr=&/FirePetal_AutoShotgun_Equip_04}}
]
properties {
 attenuationSettings=&Fort_Weapon_Reload_Att
 bOverrideConcurrency=true
 maxDistance=2250
 pitchMultiplier=1.4086
 soundClassObject=&Fort_WeaponFoley
 volumeMultiplier=0.4
 attenuationOverrides {falloffDistance=1600}
 concurrencyOverrides {resolutionRule=PreventNew}
}
