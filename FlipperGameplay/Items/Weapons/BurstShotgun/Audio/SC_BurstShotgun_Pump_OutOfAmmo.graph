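# Sound cue for the Burst Shotgun's out-of-ammo pump sound.
# Picks one of four equally weighted surround-version no-ammo waves, then applies slight random pitch variation and a fixed volume boost through a Modulator.
# An Attenuation node turns off spatialization, while the cue references the shared weapon-reload attenuation asset and the weapon foley sound class.
# A low cue volume multiplier offsets the boost, and a concurrency override prevents new instances once the limit is reached.
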SoundCue /FlipperGameplay/Items/Weapons/BurstShotgun/Audio/SC_BurstShotgun_Pump_OutOfAmmo
root=Attenuation_13
nodes [
 {id=Attenuation_13, children [Modulator_13], properties {attenuationOverrides {bSpatialize=false}}}
 {id=Modulator_13, children [Random_1], properties {pitchMax=1.1, pitchMin=0.9, volumeMax=2, volumeMin=2}}
 {id=Random_1, children [WavePlayer_0,WavePlayer_1,WavePlayer_2,WavePlayer_6], properties {weights [1,1,1,1]}}
 {
  id=WavePlayer_0
  properties {soundWaveAssetPtr=/FlipperGameplay/Items/Weapons/BurstShotgun/Audio/Srnnd_Versions/sw_BurstShotgun_NoAmmo_srnnd_01}
 }
 {
  id=WavePlayer_1
  properties {soundWaveAssetPtr=/FlipperGameplay/Items/Weapons/BurstShotgun/Audio/Srnnd_Versions/sw_BurstShotgun_NoAmmo_srnnd_02}
 }
 {
  id=WavePlayer_2
  properties {soundWaveAssetPtr=/FlipperGameplay/Items/Weapons/BurstShotgun/Audio/Srnnd_Versions/sw_BurstShotgun_NoAmmo_srnnd_03}
 }
 {
  id=WavePlayer_6
  properties {soundWaveAssetPtr=/FlipperGameplay/Items/Weapons/BurstShotgun/Audio/Srnnd_Versions/sw_BurstShotgun_NoAmmo_srnnd_04}
 }
]
properties {
 attenuationSettings=&Fort_Weapon_Reload_Att
 bOverrideConcurrency=true
 maxDistance=2097152
 soundClassObject=&Fort_WeaponFoley
 volumeMultiplier=0.1
 attenuationOverrides {falloffDistance=1600}
 concurrencyOverrides {resolutionRule=PreventNew}
}
