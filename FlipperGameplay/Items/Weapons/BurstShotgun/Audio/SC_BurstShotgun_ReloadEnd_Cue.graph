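# Sound cue for the end of the Burst Shotgun reload, routed to the weapon foley sound class.
# A random node picks one of four equally weighted reload-end waves, then a modulator applies a slight random downward pitch variation.
# It plays at reduced volume with weapon reload attenuation, and a concurrency override limits it to two instances per owner, stopping the oldest.
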SoundCue /FlipperGameplay/Items/Weapons/BurstShotgun/Audio/SC_BurstShotgun_ReloadEnd_Cue
root=Modulator_14
nodes [
 {id=Modulator_14, children [Random_0], properties {pitchMax=1, pitchMin=0.98, volumeMax=1, volumeMin=1}}
 {id=Random_0, children [WavePlayer_0,WavePlayer_2,WavePlayer_3,WavePlayer_4], properties {weights [1,1,1,1]}}
 {id=WavePlayer_0, properties {soundWaveAssetPtr=&/BurstShotgun_ReloadEnd_01}}
 {id=WavePlayer_2, properties {soundWaveAssetPtr=&/BurstShotgun_ReloadEnd_02}}
 {id=WavePlayer_3, properties {soundWaveAssetPtr=&/BurstShotgun_ReloadEnd_03}}
 {id=WavePlayer_4, properties {soundWaveAssetPtr=&/BurstShotgun_ReloadEnd_04}}
]
properties {
 attenuationSettings=&Fort_Weapon_Reload_Att
 bOverrideConcurrency=true
 maxDistance=2250
 pitchMultiplier=1.0324
 soundClassObject=&Fort_WeaponFoley
 volumeMultiplier=0.4
 attenuationOverrides {falloffDistance=1600}
 concurrencyOverrides {bLimitToOwner=true, maxCount=2, resolutionRule=StopOldest}
}
