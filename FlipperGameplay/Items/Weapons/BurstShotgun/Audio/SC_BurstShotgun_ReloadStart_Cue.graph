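# Sound cue for the start of the Burst Shotgun reload, classed as weapon foley.
# A random node picks one of three equally weighted surround reload-start waves, then a modulator applies a fixed volume boost under a low cue volume multiplier.
# It references the shared weapon reload attenuation settings and overrides concurrency with a rule that prevents new instances once the limit is reached.
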SoundCue /FlipperGameplay/Items/Weapons/BurstShotgun/Audio/SC_BurstShotgun_ReloadStart_Cue
root=Modulator_0
nodes [
 {id=Modulator_0, children [Random_1], properties {volumeMax=2, volumeMin=2}}
 {id=Random_1, children [WavePlayer_0,WavePlayer_4,WavePlayer_5], properties {weights [1,1,1]}}
 {
  id=WavePlayer_0
  properties {soundWaveAssetPtr=/FlipperGameplay/Items/Weapons/BurstShotgun/Audio/Srnnd_Versions/sw_BurstShotgun_ReloadStart_srnnd_01}
 }
 {
  id=WavePlayer_4
  properties {soundWaveAssetPtr=/FlipperGameplay/Items/Weapons/BurstShotgun/Audio/Srnnd_Versions/sw_BurstShotgun_ReloadStart_srnnd_02}
 }
 {
  id=WavePlayer_5
  properties {soundWaveAssetPtr=/FlipperGameplay/Items/Weapons/BurstShotgun/Audio/Srnnd_Versions/sw_BurstShotgun_ReloadStart_srnnd_03}
 }
]
properties {
 attenuationSettings=&Fort_Weapon_Reload_Att
 bOverrideConcurrency=true
 maxDistance=2250
 pitchMultiplier=0.992
 soundClassObject=&Fort_WeaponFoley
 volumeMultiplier=0.09
 attenuationOverrides {falloffDistance=1600}
 concurrencyOverrides {resolutionRule=PreventNew}
}
